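# Editor utility widget for a single cell of the PCG mesh grass heatmap tool, drawn as an aspect-ratio-constrained button.
# On Construct, it tints the button with the Color variable and sets the tooltip to the Info string.
# Releasing the button moves the level viewport camera to TargetLocation through UnrealEditorSubsystem, keeping the current camera rotation.
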EditorUtilityWidgetBlueprint /Game/Athena/Helios/Environments/Landscape/WorldBuildingTools/PCGGrass/Utilities/Heatmap/EUW_MeshGrass_Heatmap_Panel
graphs [EventGraph {nodes [
   "On Released (EditorUtilityButton_40)"() {is=title K2Node_ComponentBoundEvent}
   n1 = Get UnrealEditorSubsystem()
   > SetLevelViewportCameraInfo!(
    execute
    self:Unreal Editor Subsystem Object Reference=@n1
    CameraLocation:Vector=Get TargetLocation() -> (TargetLocation:Vector)
    CameraRotation:Rotator=GetLevelViewportCameraInfo!(self:Unreal Editor Subsystem Object Reference=@n1) -> (CameraLocation:Vector, CameraRotation:Rotator, ReturnValue:bool).CameraRotation
   ) -> (then)
   Event Construct()
   > SetBackgroundColor!(
    execute
    self:Button Object Reference=Get EditorUtilityButton_40() -> (EditorUtilityButton_40:Editor Utility Button Object Reference)
    InBackgroundColor:Linear Color Structure=Get Color() -> (Color:Linear Color Structure)
   ) -> (then)
   > SetToolTipText!(execute, self:Widget Object Reference, InToolTipText:"Text (by ref)"="To Text (String)"(InString=Get Info() -> (Info:String))) -> (then)
  ]}]
variables [
 Color {type=LinearColor, default=<rgba 0 0 0 0>}
 Info {type=string, default=""}
 TargetLocation {type=Vector}
]
widgets {
 info {rootWidgetClass=&SizeBox, widgetCount=2}
 widgets [
  SizeBox_0 {widgetClassPath=&SizeBox, widget=SizeBox_0, properties {bOverride_MaxAspectRatio=true, bOverride_MinAspectRatio=true}}
  EditorUtilityButton_40 {
   widgetClassPath=&EditorUtilityButton
   parent=SizeBox_0
   widget=EditorUtilityButton_40
   properties {widgetStyle {
     clickedSlateSound {resourceObject=null}
     disabled {
      bIsDynamicallyLoaded=false
      drawAs=RoundedBox
      imageType=NoImage
      mirroring=NoMirror
      resourceName=None
      resourceObject=null
      tiling=NoTile
      imageSize=<0 0>
      margin=<ltrb 0 0 0 0>
      outlineSettings {
       bUseBrushTransparency=false
       roundingType=FixedRadius
       width=1
       color {colorUseRule=UseColor_ColorTable, specifiedColor=<rgba 0 0 0 0>}
       cornerRadii=<4 4 4 4>
      }
      tintColor {colorUseRule=UseColor_Specified, specifiedColor=<rgba 0.0395 0.0395 0.0395 1>}
      uVRegion {bIsValid=false, max=<0 0>, min=<0 0>}
     }
     disabledForeground {colorUseRule=UseColor_ColorTable, specifiedColor=<rgba 0 0 0 0>}
     hovered {
      bIsDynamicallyLoaded=false
      drawAs=RoundedBox
      imageType=NoImage
      mirroring=NoMirror
      resourceName=None
      resourceObject=null
      tiling=NoTile
      imageSize=<32 32>
      margin=<ltrb 0 0 0 0>
      outlineSettings {
       bUseBrushTransparency=false
       roundingType=FixedRadius
       width=1
       color {colorUseRule=UseColor_ColorTable, specifiedColor=<rgba 0 0 0 0>}
       cornerRadii=<4 4 4 4>
      }
      tintColor {colorUseRule=UseColor_Specified, specifiedColor=<rgba 0.9 0.9 0.9 1>}
      uVRegion {bIsValid=false, max=<0 0>, min=<0 0>}
     }
     hoveredForeground {colorUseRule=UseColor_ColorTable, specifiedColor=<rgba 0 0 0 0>}
     hoveredSlateSound {resourceObject=null}
     normal {
      bIsDynamicallyLoaded=false
      drawAs=RoundedBox
      imageType=NoImage
      mirroring=NoMirror
      resourceName=None
      resourceObject=null
      tiling=NoTile
      imageSize=<32 32>
      margin=<ltrb 0 0 0 0>
      outlineSettings {
       bUseBrushTransparency=false
       roundingType=FixedRadius
       width=1
       color {colorUseRule=UseColor_ColorTable, specifiedColor=<rgba 0 0 0 0>}
       cornerRadii=<4 4 4 4>
      }
      tintColor {colorUseRule=UseColor_Specified, specifiedColor=<rgba 1 1 1 1>}
      uVRegion {bIsValid=false, max=<0 0>, min=<0 0>}
     }
     normalForeground {colorUseRule=UseColor_ColorTable, specifiedColor=<rgba 0 0 0 0>}
     normalPadding=<ltrb 12 1.5 12 1.5>
     pressed {
      bIsDynamicallyLoaded=false
      drawAs=RoundedBox
      imageType=NoImage
      mirroring=NoMirror
      resourceName=None
      resourceObject=null
      tiling=NoTile
      imageSize=<32 32>
      margin=<ltrb 0 0 0 0>
      outlineSettings {
       bUseBrushTransparency=false
       roundingType=FixedRadius
       width=1
       color {colorUseRule=UseColor_ColorTable, specifiedColor=<rgba 0 0 0 0>}
       cornerRadii=<4 4 4 4>
      }
      tintColor {colorUseRule=UseColor_Specified, specifiedColor=<rgba 0.0284 0.0284 0.0284 1>}
      uVRegion {bIsValid=false, max=<0 0>, min=<0 0>}
     }
     pressedForeground {colorUseRule=UseColor_ColorTable, specifiedColor=<rgba 0 0 0 0>}
     pressedPadding=<ltrb 12 2.5 12 0.5>
     pressedSlateSound {resourceObject=null}
    }}
   slot {class=SizeBoxSlot}
  }
 ]
}
